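# Masked, lit Niagara sprite material that renders dirt bits for the Lazy Lizz pickaxe effects.
# A scalable tiling dirt texture uses its own alpha as bump-offset height, adding parallax to color and normal samples.
# Particle color tints the texture and feeds an emissive-adjustment function.
# Particle alpha erodes the texture alpha by subtraction, and a centered sphere mask rounds the sprite edges.
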Material /BRCosmetics/Weapons/FORT_Melee/Pickaxe_LazyLizz/FX/M_Pickaxe_LazyLizz_DirtBits
graph {
 properties {bUsedWithNiagaraSprites=true, blendMode=BLEND_Masked, pixelDepthOffsetMode=PDOM_Legacy, usedShadingModels=Default Lit}
 nodes [
  n1 = ParticleColor()
  n2 = BumpOffset(
   Coordinate=Multiply(A=TextureCoordinate(), B=ScalarParameter() {defaultValue=1, parameterName=Texture Scale, sliderMax=5})
   Height=TextureSample() {samplerType=SAMPLERTYPE_LinearGrayscale, texture=/Game/Effects/Fort_Effects/Textures/Rocks/T_Tiling_Dust_Dirt_Bits_M}.A
  ) {heightRatio=0.02}
  n3 = TextureSample(UVs=@n2) {samplerType=SAMPLERTYPE_LinearGrayscale, texture=&T_Tiling_Dust_Dirt_Bits_M}
  MaterialFunctionCall(Emissive=Multiply(A=@n1.RGB, B=@n3.RGB)) {materialFunction=&MF_FX_EmissiveAdjustments}
  n4 = Constant() {r=0.5}
  Multiply(A=Subtract(A=@n3.A, B=OneMinus(None=@n1.A)), B=SphereMask(A=TextureCoordinate(), B=@n4, Radius=@n4) {hardnessPercent=30})
  Reroute(None=TextureSample(UVs=@n2) {samplerType=SAMPLERTYPE_Normal, texture=/Game/Effects/Fort_Effects/Textures/Rocks/T_Tiling_Dust_Dirt_Bits_N}.RGB)
 ]
}
