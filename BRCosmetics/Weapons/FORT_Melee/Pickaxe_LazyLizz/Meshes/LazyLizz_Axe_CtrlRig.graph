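# Control Rig for the LazyLizz axe pickaxe, which position-constrains its six dynamic leaf bones.
# A single execution chain runs six position constraints with maintained offset, at full weight on all axes, targeting transform bones under whip_skin_16.
# Leaves 02–06 follow their own parent transforms, while leaf_01_dyn follows transform33, leaf_02's parent, rather than its parent transform32.
# Other dynamic bones and the partial body skeleton in the hierarchy are untouched by the graph.
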ControlRigBlueprint /BRCosmetics/Weapons/FORT_Melee/Pickaxe_LazyLizz/Meshes/LazyLizz_Axe_CtrlRig
graphs [
 RigVMFunctionLibrary {links [], nodes []}
 RigVMModel {nodes [
   RigUnit_BeginExecution = &RigUnit_BeginExecution()
   /Script/ControlRig.RigUnit_PositionConstraintLocalSpaceOffset(
    Child=(Type=Bone,Name="leaf_01_dyn")
    Filter=(bX=True,bY=True,bZ=True)
    Parents=((Item=(Type=Bone,Name="transform33"),Weight=1))
    Weight=1
    bMaintainOffset=True
    ExecutePin=@RigUnit_BeginExecution.ExecutePin
   ) {is=struct RigVMUnitNode}
   > &RigUnit_PositionConstraintLocalSpaceOffset(
    Child=(Type=Bone,Name="leaf_02_dyn")
    Filter=(bX=True,bY=True,bZ=True)
    Parents=((Item=(Type=Bone,Name="transform33"),Weight=1))
    Weight=1
    bMaintainOffset=True
   )
   > &RigUnit_PositionConstraintLocalSpaceOffset(
    Child=(Type=Bone,Name="leaf_03_dyn")
    Filter=(bX=True,bY=True,bZ=True)
    Parents=((Item=(Type=Bone,Name="transform34"),Weight=1))
    Weight=1
    bMaintainOffset=True
   )
   > &RigUnit_PositionConstraintLocalSpaceOffset(
    Child=(Type=Bone,Name="leaf_04_dyn")
    Filter=(bX=True,bY=True,bZ=True)
    Parents=((Item=(Type=Bone,Name="transform35"),Weight=1))
    Weight=1
    bMaintainOffset=True
   )
   > &RigUnit_PositionConstraintLocalSpaceOffset(
    Child=(Type=Bone,Name="leaf_05_dyn")
    Filter=(bX=True,bY=True,bZ=True)
    Parents=((Item=(Type=Bone,Name="transform36"),Weight=1))
    Weight=1
    bMaintainOffset=True
   )
   > &RigUnit_PositionConstraintLocalSpaceOffset(
    Child=(Type=Bone,Name="leaf_06_dyn")
    Filter=(bX=True,bY=True,bZ=True)
    Parents=((Item=(Type=Bone,Name="transform37"),Weight=1))
    Weight=1
    bMaintainOffset=True
   )
  ]}
]
hierarchy [
 root {loc=<0 0 0>, rot=<0 0 0 1>}
 b_AR_Root {parent=root, loc=<0 0 0>, rot=<0 0 0 1>}
 Melee_TwoHanded_Handle {parent=b_AR_Root, loc=<0 0 0>, rot=<0 0 0.7071 -0.7071>}
 whip_skin_01 {parent=Melee_TwoHanded_Handle, loc=<0.1016 0.0001 3.2092>, rot=<0.5 -0.5 0.5 0.5>}
 whip_skin_02 {parent=whip_skin_01, loc=<2.799 0 0>, rot=<0 0 0 -1>}
 whip_skin_03 {parent=whip_skin_02, loc=<2.799 0 0>, rot=<0 0 0 -1>}
 whip_skin_04 {parent=whip_skin_03, loc=<2.799 0 0>, rot=<0 0 0 -1>}
 whip_skin_05 {parent=whip_skin_04, loc=<2.7989 0 0>, rot=<0 0 0 -1>}
 whip_skin_06 {parent=whip_skin_05, loc=<2.7988 0 0>, rot=<0 0 0 -1>}
 whip_skin_07 {parent=whip_skin_06, loc=<2.7986 0 0>, rot=<0 0 0 -1>}
 whip_skin_08 {parent=whip_skin_07, loc=<2.7982 0 0>, rot=<0 0 0 -1>}
 whip_skin_09 {parent=whip_skin_08, loc=<2.7975 0 0>, rot=<0 0 0 -1>}
 whip_skin_10 {parent=whip_skin_09, loc=<2.7975 0 0>, rot=<0 0 0 -1>}
 whip_skin_11 {parent=whip_skin_10, loc=<2.7975 0 0>, rot=<0 0 0 -1>}
 whip_skin_12 {parent=whip_skin_11, loc=<2.7975 0 0>, rot=<0 0 0 -1>}
 whip_skin_13 {parent=whip_skin_12, loc=<2.7975 0 0>, rot=<0 0 0 -1>}
 whip_skin_14 {parent=whip_skin_13, loc=<2.7975 0 0>, rot=<0 0 0 -1>}
 whip_skin_15 {parent=whip_skin_14, loc=<2.7975 0 0>, rot=<0 0 0 -1>}
 whip_skin_16 {parent=whip_skin_15, loc=<2.7975 0 0>, rot=<0 0 0 -1>}
 whip_skin_17 {parent=whip_skin_16, loc=<2.7975 0 0>, rot=<0 0 0 -1>}
 whip_skin_18 {parent=whip_skin_17, loc=<2.7975 0 0>, rot=<0 0 0 1>}
 whip_skin_19 {parent=whip_skin_18, loc=<2.7975 0 0>, rot=<0 0 0 1>}
 whip_skin_20 {parent=whip_skin_19, loc=<2.7975 0 0>, rot=<0 0 0 1>}
 group2 {parent=whip_skin_19, loc=<2.7975 0 0>, rot=<0 0 0 1>}
 root_03_dyn_parent_grp {parent=group2, loc=<1.5433 9.6225 -4.0036>, rot=<0.5318 -0.466 0.5318 -0.466>}
 root_03_dyn_anim_grp {parent=root_03_dyn_parent_grp, loc=<0 0 0>, rot=<0 0 0 1>}
 root_03_dyn_anim {parent=root_03_dyn_anim_grp, loc=<0 0 0>, rot=<0 0 0 1>}
 root_02_dyn_parent_grp {parent=group2, loc=<3.5455 11.1161 -1.5813>, rot=<0.3536 -0.6124 0.3536 -0.6124>}
 root_02_dyn_anim_grp {parent=root_02_dyn_parent_grp, loc=<0 0 0>, rot=<0 0 0 1>}
 root_02_dyn_anim {parent=root_02_dyn_anim_grp, loc=<0 0 0>, rot=<0 0 0 1>}
 root_01_dyn_parent_grp {parent=group2, loc=<2.7545 10.7105 1.3615>, rot=<0.5 -0.5 0.5 -0.5>}
 root_01_dyn_anim_grp {parent=root_01_dyn_parent_grp, loc=<0 0 0>, rot=<0 0 0 1>}
 root_01_dyn_anim {parent=root_01_dyn_anim_grp, loc=<0 0 0>, rot=<0 0 0 1>}
 transform29 {parent=whip_skin_19, loc=<-53.5928 0.1016 0.0001>, rot=<0 -0.7071 0 -0.7071>}
 root_01_base {parent=transform29, loc=<-1.3615 10.6088 59.1447>, rot=<0 0 0.7071 -0.7071>}
 root_01_dyn {parent=root_01_base, loc=<0 0 0>, rot=<0 0 0 1>}
 transform30 {parent=whip_skin_19, loc=<-53.5928 0.1016 0.0001>, rot=<0 -0.7071 0 -0.7071>}
 root_03_base {parent=transform30, loc=<4.0037 9.5208 57.9335>, rot=<0 0 0.7521 -0.659>}
 root_03_dyn {parent=root_03_base, loc=<0 0 0>, rot=<0 0 0 1>}
 transform31 {parent=whip_skin_19, loc=<-53.5928 0.1016 0.0001>, rot=<0 -0.7071 0 -0.7071>}
 root_02_base {parent=transform31, loc=<1.5814 11.0144 59.9356>, rot=<0 0 0.5 -0.866>}
 root_02_dyn {parent=root_02_base, loc=<0 0 0>, rot=<0 0 0 1>}
 transform27 {parent=whip_skin_16, loc=<-45.1943 0.1016 0.0001>, rot=<0 -0.7071 0 -0.7071>}
 mushroom_base {parent=transform27, loc=<-1.9116 2.5636 46.2545>, rot=<0.2952 0.3927 0.5624 -0.6651>}
 mushroom_dyn {parent=mushroom_base, loc=<0 0 0>, rot=<0 0 0 1>}
 transform28 {parent=whip_skin_16, loc=<-45.1943 0.1016 0.0001>, rot=<0 -0.7071 0 -0.7071>}
 sign_base {parent=transform28, loc=<4.7505 -6.0653 45.6736>, rot=<-0.3573 0.5286 -0.4868 -0.5967>}
 sign_dyn {parent=sign_base, loc=<0 0 0>, rot=<0 0 0 1>}
 group1 {parent=whip_skin_16, loc=<0 0 0>, rot=<0 0 0 1>}
 mushroom_dyn_parent_grp {parent=group1, loc=<1.0604 2.6652 1.9117>, rot=<0.6064 -0.1926 0.189 -0.748>}
 mushroom_dyn_anim_grp {parent=mushroom_dyn_parent_grp, loc=<0 0 0>, rot=<0 0 0 1>}
 mushroom_dyn_anim {parent=mushroom_dyn_anim_grp, loc=<0 0 0>, rot=<0 0 0 1>}
 sign_dyn_parent_grp {parent=group1, loc=<0.4795 -5.9636 -4.7505>, rot=<-0.5968 -0.0481 -0.0915 -0.7957>}
 sign_dyn_anim_grp {parent=sign_dyn_parent_grp, loc=<0 0 0>, rot=<0 0 0 1>}
 sign_dyn_anim {parent=sign_dyn_anim_grp, loc=<0 0 0>, rot=<0 0 0 1>}
 leaf_06_dyn_parent_grp {parent=group1, loc=<-1.663 -5.8858 8.5836>, rot=<-0.1981 -0.6788 -0.1981 -0.6788>}
 leaf_06_dyn_anim_grp {parent=leaf_06_dyn_parent_grp, loc=<0 0 0>, rot=<0 0 0 1>}
 leaf_06_dyn_anim {parent=leaf_06_dyn_anim_grp, loc=<0 0 0>, rot=<0 0 0 1>}
 leaf_05_dyn_parent_grp {parent=group1, loc=<-1.9049 -6.9827 0.1723>, rot=<-0.4217 -0.4109 -0.5077 -0.629>}
 leaf_05_dyn_anim_grp {parent=leaf_05_dyn_parent_grp, loc=<0 0 0>, rot=<0 0 0 1>}
 leaf_05_dyn_anim {parent=leaf_05_dyn_anim_grp, loc=<0 0 0>, rot=<0 0 0 1>}
 leaf_04_dyn_parent_grp {parent=group1, loc=<-2.747 -3.4356 -6.8464>, rot=<0.7215 -0.1413 0.5666 0.3719>}
 leaf_04_dyn_anim_grp {parent=leaf_04_dyn_parent_grp, loc=<0 0 0>, rot=<0 0 0 1>}
 leaf_04_dyn_anim {parent=leaf_04_dyn_anim_grp, loc=<0 0 0>, rot=<0 0 0 1>}
 leaf_03_dyn_parent_grp {parent=group1, loc=<-1.608 4.3196 -6.5285>, rot=<0.684 -0.1793 0.684 -0.1793>}
 leaf_03_dyn_anim_grp {parent=leaf_03_dyn_parent_grp, loc=<0 0 0>, rot=<0 0 0 1>}
 leaf_03_dyn_anim {parent=leaf_03_dyn_anim_grp, loc=<0 0 0>, rot=<0 0 0 1>}
 leaf_02_dyn_parent_grp {parent=group1, loc=<-1.5261 8.0318 0.2357>, rot=<0.4987 -0.5013 0.4987 -0.5013>}
 leaf_02_dyn_anim_grp {parent=leaf_02_dyn_parent_grp, loc=<0 0 0>, rot=<0 0 0 1>}
 leaf_02_dyn_anim {parent=leaf_02_dyn_anim_grp, loc=<0 0 0>, rot=<0 0 0 1>}
 leaf_01_dyn_parent_grp {parent=group1, loc=<-2.3537 3.9834 7.685>, rot=<0.169 -0.6866 0.169 -0.6866>}
 leaf_01_dyn_anim_grp {parent=leaf_01_dyn_parent_grp, loc=<0 0 0>, rot=<0 0 0 1>}
 leaf_01_dyn_anim {parent=leaf_01_dyn_anim_grp, loc=<0 0 0>, rot=<0 0 0 1>}
 transform32 {parent=whip_skin_16, loc=<-2.3537 3.9834 7.685>, rot=<0.169 -0.6866 0.169 -0.6866>}
 leaf_01_dyn {parent=transform32, loc=<0 0 0>, rot=<0 0 0 1>}
 transform33 {parent=whip_skin_16, loc=<-1.5261 8.0318 0.2357>, rot=<0.4987 -0.5013 0.4987 -0.5013>}
 leaf_02_dyn {parent=transform33, loc=<0 0 0>, rot=<0 0 0 1>}
 transform34 {parent=whip_skin_16, loc=<-1.608 4.3196 -6.5285>, rot=<0.684 -0.1793 0.684 -0.1793>}
 leaf_03_dyn {parent=transform34, loc=<0 0 0>, rot=<0 0 0 1>}
 transform35 {parent=whip_skin_16, loc=<-2.747 -3.4356 -6.8464>, rot=<0.7215 -0.1413 0.5666 0.3719>}
 leaf_04_dyn {parent=transform35, loc=<0 0 0>, rot=<0 0 0 1>}
 transform36 {parent=whip_skin_16, loc=<-1.9049 -6.9827 0.1723>, rot=<-0.4217 -0.4109 -0.5077 -0.629>}
 leaf_05_dyn {parent=transform36, loc=<0 0 0>, rot=<0 0 0 1>}
 transform37 {parent=whip_skin_16, loc=<-1.663 -5.8858 8.5836>, rot=<-0.1981 -0.6788 -0.1981 -0.6788>}
 leaf_06_dyn {parent=transform37, loc=<0 0 0>, rot=<0 0 0 1>}
 hand_leaf_01_base {parent=Melee_TwoHanded_Handle, loc=<1.7787 0.04 -21.4071>, rot=<0.5998 -0.3738 -0.3742 -0.6004>}
 hand_leaf_01_dyn {parent=hand_leaf_01_base, loc=<0 0 0>, rot=<0 0 0 1>}
 hand_leaf_01_end_dyn {parent=hand_leaf_01_dyn, loc=<11.4185 0 0>, rot=<0 0 0 1>}
 hand_leaf_02_base {parent=Melee_TwoHanded_Handle, loc=<1.6023 -1.0773 -25.2202>, rot=<0.3506 -0.6049 -0.0229 -0.7146>}
 hand_leaf_02_dyn {parent=hand_leaf_02_base, loc=<0 0 0>, rot=<0 0 0 1>}
 hand_leaf_02_end_dyn {parent=hand_leaf_02_dyn, loc=<6.2131 0 0>, rot=<0 0 0 1>}
 hand_leaf_03_base {parent=Melee_TwoHanded_Handle, loc=<-0.3706 -0.7333 -21.2457>, rot=<-0.2132 -0.6703 0.3942 -0.5915>}
 hand_leaf_03_dyn {parent=hand_leaf_03_base, loc=<0 0 0>, rot=<0 0 0 1>}
 hand_leaf_03_end_dyn {parent=hand_leaf_03_dyn, loc=<10.1758 0 0>, rot=<0 0 0 1>}
 hand_leaf_04_base {parent=Melee_TwoHanded_Handle, loc=<-1.3379 -0.0449 -21.413>, rot=<-0.3826 -0.5985 0.593 -0.3791>}
 hand_leaf_04_dyn {parent=hand_leaf_04_base, loc=<0 0 0>, rot=<0 0 0 1>}
 hand_leaf_04_end_dyn {parent=hand_leaf_04_dyn, loc=<11.4216 0 0>, rot=<0 0 0 1>}
 hand_leaf_05_base {parent=Melee_TwoHanded_Handle, loc=<0.033 0.8717 -21.2475>, rot=<0.7139 0.1123 -0.6811 -0.1177>}
 hand_leaf_05_dyn {parent=hand_leaf_05_base, loc=<0 0 0>, rot=<0 0 0 1>}
 hand_leaf_05_end_dyn {parent=hand_leaf_05_dyn, loc=<10.2497 0 0>, rot=<0 0 0 1>}
 hand_leaf_06_base {parent=Melee_TwoHanded_Handle, loc=<0.7427 -0.1893 -28.313>, rot=<-0.3959 0.5859 -0.3959 -0.5859>}
 hand_leaf_06_dyn {parent=hand_leaf_06_base, loc=<0 0 0>, rot=<0 0 0 1>}
 hand_leaf_06_end_dyn {parent=hand_leaf_06_dyn, loc=<-6.632 0 0>, rot=<0 0 0 1>}
 body_root {parent=root, loc=<0 0 0>, rot=<0 0 0 1>}
 pelvis {parent=body_root, loc=<0 -0.7071 93.1191>, rot=<0.0127 0.707 -0.0127 -0.707>}
 thigh_l {parent=pelvis, loc=<-2.5693 -0.0109 -7.7501>, rot=<0.0012 0.0572 0.0155 -0.9982>}
 calf_l {parent=thigh_l, loc=<-39.0484 0 0>, rot=<0.0006 -0.0161 -0.0391 -0.9991>}
 thigh_r {parent=pelvis, loc=<-2.5691 -0.0109 7.7501>, rot=<0.0572 -0.0012 0.9982 0.0155>}
 calf_r {parent=thigh_r, loc=<39.0487 0 0>, rot=<0.0006 -0.0161 -0.0391 -0.9991>}
 spine_01 {parent=pelvis, loc=<6.7494 -0.6388 0>, rot=<0 0 0.034 -0.9994>}
 spine_02 {parent=spine_01, loc=<7.2518 -0.035 0>, rot=<0 0 0.0238 -0.9997>}
 spine_03 {parent=spine_02, loc=<6.0104 0 0>, rot=<0 0 0.0149 -0.9999>}
 spine_04 {parent=spine_03, loc=<8.1024 0 0>, rot=<0 0 -0.0132 -0.9999>}
 spine_05 {parent=spine_04, loc=<10.0146 0 0>, rot=<0 0 -0.0015 -1>}
 dyn_root_simspace {parent=root, loc=<0 0 0>, rot=<0 0 0 1>}
]
