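# Physics asset for the LazyLizz two-handed pickaxe, defining capsule bodies and joint limits for its flexible and dangling parts.
# A 14-link whip_skin chain hangs from the handle, ending in a multi-capsule body that carries leaf, root, mushroom and sign bodies.
# Six separate hand-leaf chains use tight swing limits and mostly zero twist.
# Spine, thigh and calf capsules are also included, with free-rotating thigh-to-calf joints.
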PhysicsAsset /BRCosmetics/Weapons/FORT_Melee/Pickaxe_LazyLizz/Meshes/LazyLizz_Axe_Physics2
bodies {
 Melee_TwoHanded_Handle [{
   shapeType=Capsule
   length=0.9287
   radius=40.7774
   shapeName=Melee_TwoHanded_Handle_capsule
   center=<0.4061 0.002 -37.5903>
   rotation=<rot -0.0287 -90 -0.6551>
  }]
 calf_l [{shapeType=Capsule, length=50, radius=8, center=<-30 0 0>, rotation=<rot 90 0 0>}]
 calf_r [{shapeType=Capsule, length=50, radius=8, center=<30 0 0>, rotation=<rot 90 0 0>}]
 hand_leaf_01_base [{shapeType=Capsule, length=0.505, radius=0.505, shapeName=hand_leaf_01_base_capsule, rotation=<rot 0 -90 0>}]
 hand_leaf_01_dyn [{shapeType=Capsule, length=3.7242, radius=3.0315, shapeName=hand_leaf_01_dyn_capsule, center=<3 0 0>, rotation=<rot 0 90.7152 -89.9327>}]
 hand_leaf_01_end_dyn [{shapeType=Capsule, length=7.0839, radius=3.1234, shapeName=hand_leaf_01_end_dyn_capsule, center=<3 0 0>, rotation=<rot 0 -115.4477 -89.9861>}]
 hand_leaf_02_base [{shapeType=Capsule, length=0.505, radius=0.505, shapeName=hand_leaf_02_base_capsule, rotation=<rot 0 -90 0>}]
 hand_leaf_02_dyn [{shapeType=Capsule, length=3.978, radius=2.9221, shapeName=hand_leaf_02_dyn_capsule, center=<3 0 0>, rotation=<rot 90 0 0>}]
 hand_leaf_02_end_dyn [{shapeType=Capsule, length=6.8594, radius=2.835, shapeName=hand_leaf_02_end_dyn_capsule, center=<3 0 0>, rotation=<rot 90 27.4591 55.2573>}]
 hand_leaf_03_base [{shapeType=Capsule, length=0.505, radius=0.505, shapeName=hand_leaf_03_base_capsule, rotation=<rot 0 -90 0>}]
 hand_leaf_03_dyn [{shapeType=Capsule, length=4.2704, radius=2.7356, shapeName=hand_leaf_03_dyn_capsule, center=<3 0 0>, rotation=<rot 90 0 0>}]
 hand_leaf_03_end_dyn [{shapeType=Capsule, length=6.5886, radius=2.8242, shapeName=hand_leaf_03_end_dyn_capsule, center=<3 0 0>, rotation=<rot 90 16.7835 24.221>}]
 hand_leaf_04_base [{shapeType=Capsule, length=0.505, radius=0.505, shapeName=hand_leaf_04_base_capsule, rotation=<rot 0 -90 0>}]
 hand_leaf_04_dyn [{shapeType=Capsule, length=4.8809, radius=3.1001, shapeName=hand_leaf_04_dyn_capsule, center=<3 0 0>, rotation=<rot 90 0 0>}]
 hand_leaf_04_end_dyn [{shapeType=Capsule, length=7.3063, radius=3.1413, shapeName=hand_leaf_04_end_dyn_capsule, center=<3 0 0>, rotation=<rot 90 0 0>}]
 hand_leaf_05_base [{shapeType=Capsule, length=0.505, radius=0.505, shapeName=hand_leaf_05_base_capsule, rotation=<rot 0 -90 0>}]
 hand_leaf_05_dyn [{shapeType=Capsule, length=4.3109, radius=2.6484, shapeName=hand_leaf_05_dyn_capsule, center=<3 0 0>, rotation=<rot 90 0 0>}]
 hand_leaf_05_end_dyn [{shapeType=Capsule, length=6.7415, radius=2.7314, shapeName=hand_leaf_05_end_dyn_capsule, center=<3 0 0>, rotation=<rot 90 0 0>}]
 hand_leaf_06_base [{shapeType=Capsule, length=0.505, radius=0.505, shapeName=hand_leaf_06_base_capsule, rotation=<rot 0 -90 0>}]
 hand_leaf_06_dyn [{shapeType=Capsule, length=3.8119, radius=2.7787, shapeName=hand_leaf_06_dyn_capsule, center=<-3 0 0>, rotation=<rot 90 0 0>}]
 hand_leaf_06_end_dyn [{shapeType=Capsule, length=4.6098, radius=2.8143, shapeName=hand_leaf_06_end_dyn_capsule, center=<-3 0 0>, rotation=<rot 90 0 0>}]
 leaf_01_dyn [{shapeType=Capsule, length=5.8688, radius=2.7189, shapeName=leaf_01_dyn_capsule, center=<-2 0 0>, rotation=<rot 90 0 0>}]
 leaf_02_dyn [{shapeType=Capsule, length=6.5762, radius=2.0696, shapeName=leaf_02_dyn_capsule, center=<-3 0 -0.5>, rotation=<rot 0 0 90>}]
 leaf_03_dyn [{
   shapeType=Capsule
   length=10.0664
   radius=1.932
   shapeName=leaf_03_dyn_capsule
   center=<-1.0634 -3.2638 -0.288>
   rotation=<rot 90 179.5524 -175.7643>
  }]
 leaf_04_dyn [{
   shapeType=Capsule
   length=6.111
   radius=1.9622
   shapeName=leaf_04_dyn_capsule
   center=<-1.9922 -2.2069 0.2105>
   rotation=<rot 50.9473 -88.1739 -59.1763>
  }]
 leaf_05_dyn [{shapeType=Capsule, length=5.9566, radius=4.6994, shapeName=leaf_05_dyn_capsule, center=<-3 0.9871 -2.5094>, rotation=<rot 90 0 0>}]
 leaf_06_dyn [{shapeType=Capsule, length=4.8048, radius=4.1372, shapeName=leaf_06_dyn_capsule, center=<-4.0445 0 0>, rotation=<rot 0 -87.3009 -81.879>}]
 mushroom_dyn [{shapeType=Capsule, length=7.4091, radius=2.4819, shapeName=mushroom_dyn_capsule, center=<-0.5 0 0>, rotation=<rot 90 0 0>}]
 root_01_dyn [{shapeType=Capsule, length=4, radius=1.867, shapeName=root_01_dyn_capsule, center=<-1 0 0>, rotation=<rot 45.2781 0 0>}]
 root_02_dyn [{shapeType=Capsule, length=5.4346, radius=1.9062, shapeName=root_02_dyn_capsule, center=<-2 0 0>, rotation=<rot 25.7459 -90 -30.7282>}]
 root_03_dyn [{
   shapeType=Capsule
   length=3.1457
   radius=1.4822
   shapeName=root_03_dyn_capsule
   center=<-2 1.4257 1.2377>
   rotation=<rot 0 -145.7397 -60.8403>
  }]
 sign_dyn [{shapeType=Capsule, length=6.5335, radius=1.8264, shapeName=sign_dyn_capsule, center=<-1 0 0>, rotation=<rot 90 0 0>}]
 spine_01 [{
   shapeType=Capsule
   length=26.9373
   radius=23.6213
   shapeName=spine_01_capsule
   center=<10.0178 -0.0008 0.2775>
   rotation=<rot 0 -90.0045 -91.5867>
  }]
 thigh_l [{shapeType=Capsule, length=40, radius=8, center=<-25 0 0>, rotation=<rot 90 0 0>}]
 thigh_r [{shapeType=Capsule, length=40, radius=8, center=<25 0 0>, rotation=<rot 90 0 0>}]
 whip_skin_01 [{shapeType=Capsule, length=2, radius=1.5, center=<0.05 0 0>, rotation=<rot 90 0 0>}]
 whip_skin_02 [{shapeType=Capsule, length=2, radius=1.5, center=<0.05 0 0>, rotation=<rot 90 0 0>}]
 whip_skin_03 [{shapeType=Capsule, length=2, radius=2, center=<0.05 0 0>, rotation=<rot 90 0 0>}]
 whip_skin_04 [{shapeType=Capsule, length=2, radius=2, center=<0.1 0 0>, rotation=<rot 90 0 0>}]
 whip_skin_05 [{shapeType=Capsule, length=2, radius=2, center=<0.1 0 0>, rotation=<rot 90 0 0>}]
 whip_skin_06 [{shapeType=Capsule, length=2, radius=2, center=<0.1 0 0>, rotation=<rot 90 0 0>}]
 whip_skin_07 [{shapeType=Capsule, length=2, radius=2, center=<0.1 0 0>, rotation=<rot 90 0 0>}]
 whip_skin_08 [{shapeType=Capsule, length=2, radius=2, center=<0.1 0 0>, rotation=<rot 90 0 0>}]
 whip_skin_09 [{shapeType=Capsule, length=2, radius=2, center=<0.1 0 0>, rotation=<rot 90 0 0>}]
 whip_skin_10 [{shapeType=Capsule, length=2, radius=2, center=<0.1 0 0>, rotation=<rot 90 0 0>}]
 whip_skin_11 [{shapeType=Capsule, length=2, radius=2, center=<0.1 0 0>, rotation=<rot 90 0 0>}]
 whip_skin_12 [{shapeType=Capsule, length=2, radius=2, center=<0.1 0 0>, rotation=<rot 90 0 0>}]
 whip_skin_13 [{shapeType=Capsule, length=1.5992, radius=2.9339, center=<0.1 0 0>, rotation=<rot 90 0 0>}]
 whip_skin_14 [
  {shapeType=Capsule, length=2.0257, radius=12.2012, center=<17.551 0 0>, rotation=<rot 90 0 0>}
  {shapeType=Capsule, length=8.6324, radius=2.658, shapeName=whip_skin_14_capsule, center=<8.0871 4.1303 -8.2168>, rotation=<rot 0 0 63.313>}
  {
   shapeType=Capsule
   length=9.5371
   radius=2.6483
   shapeName=whip_skin_14_capsule
   center=<7.7007 -4.8048 -9.9778>
   rotation=<rot 0 0 -59.0586>
  }
  {shapeType=Capsule, length=8.2917, radius=2.103, shapeName=whip_skin_14_capsule, center=<7.8152 11.1775 0>}
 ]
}
constraints [
 {bone1Name=hand_leaf_06_end_dyn, bone2Name=hand_leaf_06_dyn, swing1LimitDegrees=14, swing2LimitDegrees=14, twistLimitDegrees=0}
 {bone1Name=hand_leaf_06_dyn, bone2Name=hand_leaf_06_base, swing1LimitDegrees=14, swing2LimitDegrees=14, twistLimitDegrees=0}
 {bone1Name=hand_leaf_05_end_dyn, bone2Name=hand_leaf_05_dyn, swing1LimitDegrees=15, swing2LimitDegrees=10, twistLimitDegrees=0}
 {bone1Name=hand_leaf_05_dyn, bone2Name=hand_leaf_05_base, swing1LimitDegrees=12, swing2LimitDegrees=10, twistLimitDegrees=0}
 {bone1Name=hand_leaf_04_end_dyn, bone2Name=hand_leaf_04_dyn, swing1LimitDegrees=15, swing2LimitDegrees=10, twistLimitDegrees=0}
 {bone1Name=hand_leaf_04_dyn, bone2Name=hand_leaf_04_base, swing1LimitDegrees=13, swing2LimitDegrees=0, twistLimitDegrees=0}
 {bone1Name=hand_leaf_03_end_dyn, bone2Name=hand_leaf_03_dyn, swing1LimitDegrees=15, swing2LimitDegrees=15, twistLimitDegrees=0}
 {bone1Name=hand_leaf_03_dyn, bone2Name=hand_leaf_03_base, swing1LimitDegrees=7, swing2LimitDegrees=0, twistLimitDegrees=0}
 {bone1Name=hand_leaf_02_end_dyn, bone2Name=hand_leaf_02_dyn, swing1LimitDegrees=15, swing2LimitDegrees=15, twistLimitDegrees=0}
 {bone1Name=hand_leaf_02_dyn, bone2Name=hand_leaf_02_base, swing1LimitDegrees=5, swing2LimitDegrees=10, twistLimitDegrees=0}
 {bone1Name=hand_leaf_01_end_dyn, bone2Name=hand_leaf_01_dyn, swing1LimitDegrees=15, swing2LimitDegrees=10, twistLimitDegrees=5}
 {bone1Name=hand_leaf_01_dyn, bone2Name=hand_leaf_01_base, swing1LimitDegrees=15, swing2LimitDegrees=10, twistLimitDegrees=0}
 {bone1Name=whip_skin_01, bone2Name=Melee_TwoHanded_Handle, swing1LimitDegrees=15, swing2LimitDegrees=15, twistLimitDegrees=0}
 {bone1Name=whip_skin_02, bone2Name=whip_skin_01, swing1LimitDegrees=15, swing2LimitDegrees=15, twistLimitDegrees=0}
 {bone1Name=whip_skin_03, bone2Name=whip_skin_02, swing1LimitDegrees=15, swing2LimitDegrees=15, twistLimitDegrees=0}
 {bone1Name=whip_skin_04, bone2Name=whip_skin_03, swing1LimitDegrees=15, swing2LimitDegrees=15, twistLimitDegrees=0}
 {bone1Name=whip_skin_05, bone2Name=whip_skin_04, swing1LimitDegrees=15, swing2LimitDegrees=15, twistLimitDegrees=0}
 {bone1Name=whip_skin_06, bone2Name=whip_skin_05, swing1LimitDegrees=15, swing2LimitDegrees=15, twistLimitDegrees=0}
 {bone1Name=whip_skin_07, bone2Name=whip_skin_06, swing1LimitDegrees=15, swing2LimitDegrees=15, twistLimitDegrees=0}
 {bone1Name=whip_skin_08, bone2Name=whip_skin_07, swing1LimitDegrees=15, swing2LimitDegrees=15, twistLimitDegrees=0}
 {bone1Name=whip_skin_09, bone2Name=whip_skin_08, swing1LimitDegrees=15, swing2LimitDegrees=15, twistLimitDegrees=0}
 {bone1Name=whip_skin_10, bone2Name=whip_skin_09, swing1LimitDegrees=15, swing2LimitDegrees=15, twistLimitDegrees=0}
 {bone1Name=whip_skin_11, bone2Name=whip_skin_10, swing1LimitDegrees=15, swing2LimitDegrees=15, twistLimitDegrees=0}
 {bone1Name=whip_skin_12, bone2Name=whip_skin_11, swing1LimitDegrees=15, swing2LimitDegrees=15, twistLimitDegrees=0}
 {bone1Name=whip_skin_13, bone2Name=whip_skin_12, swing1LimitDegrees=15, swing2LimitDegrees=15, twistLimitDegrees=0}
 {bone1Name=whip_skin_14, bone2Name=whip_skin_13, swing1LimitDegrees=15, swing2LimitDegrees=15, twistLimitDegrees=0}
 {
  bone1Name=calf_l
  bone2Name=thigh_l
  swing1LimitDegrees=45
  swing1Motion=Free
  swing2LimitDegrees=45
  swing2Motion=Free
  twistLimitDegrees=45
  twistMotion=Free
 }
 {
  bone1Name=calf_r
  bone2Name=thigh_r
  swing1LimitDegrees=45
  swing1Motion=Free
  swing2LimitDegrees=45
  swing2Motion=Free
  twistLimitDegrees=45
  twistMotion=Free
 }
 {bone1Name=root_03_dyn, bone2Name=whip_skin_14, swing1LimitDegrees=12, swing2LimitDegrees=12, twistLimitDegrees=0}
 {bone1Name=root_02_dyn, bone2Name=whip_skin_14, swing1LimitDegrees=12, swing2LimitDegrees=16, twistLimitDegrees=0}
 {bone1Name=root_01_dyn, bone2Name=whip_skin_14, swing1LimitDegrees=12, swing2LimitDegrees=14, twistLimitDegrees=0}
 {bone1Name=sign_dyn, bone2Name=whip_skin_14, swing1LimitDegrees=10, swing2LimitDegrees=10, twistLimitDegrees=5}
 {bone1Name=mushroom_dyn, bone2Name=whip_skin_14, swing1LimitDegrees=10, swing2LimitDegrees=5, twistLimitDegrees=0}
 {bone1Name=leaf_06_dyn, bone2Name=whip_skin_14, swing1LimitDegrees=10, swing2LimitDegrees=15, twistLimitDegrees=0}
 {bone1Name=leaf_05_dyn, bone2Name=whip_skin_14, swing1LimitDegrees=10, swing2LimitDegrees=15, twistLimitDegrees=0}
 {bone1Name=leaf_04_dyn, bone2Name=whip_skin_14, swing1LimitDegrees=10, swing2LimitDegrees=15, twistLimitDegrees=0}
 {bone1Name=leaf_03_dyn, bone2Name=whip_skin_14, swing1LimitDegrees=10, swing2LimitDegrees=15, twistLimitDegrees=0}
 {bone1Name=leaf_02_dyn, bone2Name=whip_skin_14, swing1LimitDegrees=10, swing2LimitDegrees=15, twistLimitDegrees=0}
 {bone1Name=leaf_01_dyn, bone2Name=whip_skin_14, swing1LimitDegrees=10, swing2LimitDegrees=15, twistLimitDegrees=0}
]
